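# Sound Cue for the looping ambient sound of the Fire Petal Shrine treasure chest, in the loot ambient loop sound class.
# A Random node picks between two equally weighted looping players, both using waves named Loop_01, feeding a Modulator that fixes pitch at 1.5x, then a single-input Mixer.
# A shared attenuation asset and a loot ambient concurrency set govern distance falloff and simultaneous instances.
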SoundCue /Game/Sounds/Foley_Loot/Containers/Treasure_Chest/SC_FirePetalShrine_Chest_Ambient_Loop
root=Mixer_0
nodes [
 {id=Mixer_0, children [Modulator_0], properties {inputVolume [1]}}
 {id=Modulator_0, children [Random_0], properties {pitchMax=1.5, pitchMin=1.5, volumeMin=1}}
 {id=Random_0, children [WavePlayer_2,WavePlayer_3], properties {bShouldExcludeFromBranchCulling=true, weights [1,1]}}
 {
  id=WavePlayer_2
  properties {bLooping=true, soundWaveAssetPtr=&/FirePetalShrine_Chest_Ambient_Loop_01}
 }
 {
  id=WavePlayer_3
  properties {bLooping=true, soundWaveAssetPtr=&FirePetalShrine_Chest_Ambient_Loop_01}
 }
]
properties {
 attenuationSettings=/Game/Sounds/Attenuation/Fort_Item_Ambient_Loop_Mono_Default_Att
 maxDistance=1250
 soundClassObject=/Game/Sounds/FortSoundClassesAndModes/Fort_Loot_Ambient_Loop
 volumeMultiplier=0.7
 attenuationOverrides {bAttenuate=false, bEnableReverbSend=false, bSpatialize=false}
 concurrencyOverrides {maxCount=3, resolutionRule=StopFarthestThenPreventNew, volumeScaleAttackTime=0.514}
 concurrencySet [/Game/Sounds/Concurrency_Rules/Loot_Ambient_Conc]
}
