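# Sound cue for opening a scenario treasure chest, using item-interact attenuation and the loot reward sound class.
# A Random node picks one of three equally weighted open waves and feeds a Mixer at boosted volume.
# The Mixer's second branch runs a short Delay into a pitch-randomizing Modulator with no child, so that layer is silent.
# A concurrency override allows one instance per owner and prevents new plays while it sounds.
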SoundCue /Game/Sounds/Foley_Loot/Containers/Treasure_Chest/Scenario_Chest_Open_Cue
root=Mixer_0
nodes [
 {id=Delay_0, children [Modulator_3], properties {delayMax=0.05, delayMin=0.05}}
 {id=Mixer_0, children [Random_6,Delay_0], properties {inputVolume [1.5,1]}}
 {id=Modulator_3, children [null], properties {pitchMax=1.5, pitchMin=1, volumeMax=0.7, volumeMin=0.7}}
 {id=Random_6, children [WavePlayer_0,WavePlayer_1,WavePlayer_2], properties {bShouldExcludeFromBranchCulling=true, weights [1,1,1]}}
 {id=WavePlayer_0, properties {soundWaveAssetPtr=&/Treasure_Chest_Scenario_Open_01}}
 {id=WavePlayer_1, properties {soundWaveAssetPtr=&/Treasure_Chest_Scenario_Open_02}}
 {id=WavePlayer_2, properties {soundWaveAssetPtr=&/Treasure_Chest_Scenario_Open_03}}
]
properties {
 attenuationSettings=/Game/Sounds/Attenuation/Fort_Item_Interact_Att
 bOverrideConcurrency=true
 maxDistance=1850
 soundClassObject=&Fort_Loot_Reward_PSM
 volumeMultiplier=1.5
 concurrencyOverrides {bLimitToOwner=true, maxCount=1, resolutionRule=PreventNew}
}
