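# Sound cue for a tiered treasure chest impact, using the small melee weapon attenuation and sound class.
# A quality-level switch at the root picks one of four branches, each passing through a Modulator.
# The first branch randomizes among three impact mixes, the next two between two, and the last plays a single mix.
# A concurrency override allows one instance at a time and blocks new plays while it is active.
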SoundCue /Game/Sounds/Foley_Loot/Containers/Treasure_Chest/Tiered_Chest_Impact_Cue
root=QualityLevel_0
nodes [
 {id=Modulator_0, children [Random_0]}
 {id=Modulator_1, children [Random_2]}
 {id=Modulator_2, children [Random_3]}
 {id=Modulator_3, children [WavePlayer_12]}
 {id=QualityLevel_0, children [Modulator_0,Modulator_1,Modulator_2,Modulator_3]}
 {id=Random_0, children [WavePlayer_0,WavePlayer_1,WavePlayer_3], properties {weights [1,1,1]}}
 {id=Random_2, children [WavePlayer_4,WavePlayer_9], properties {weights [1,1]}}
 {id=Random_3, children [WavePlayer_10, WavePlayer_11], properties {weights [1,1]}}
 {id=WavePlayer_0, properties {soundWaveAssetPtr=&/Treasure_Impact_mix_01}}
 {id=WavePlayer_1, properties {soundWaveAssetPtr=&/Treasure_Impact_mix_02}}
 {id=WavePlayer_10, properties {soundWaveAssetPtr=&Treasure_Impact_mix_01}}
 {id=WavePlayer_11, properties {soundWaveAssetPtr=&Treasure_Impact_mix_02}}
 {id=WavePlayer_12, properties {soundWaveAssetPtr=&Treasure_Impact_mix_01}}
 {id=WavePlayer_3, properties {soundWaveAssetPtr=&/Treasure_Impact_mix_03}}
 {id=WavePlayer_4, properties {soundWaveAssetPtr=&Treasure_Impact_mix_01}}
 {id=WavePlayer_9, properties {soundWaveAssetPtr=&Treasure_Impact_mix_02}}
]
properties {
 attenuationSettings=&Fort_Weapon_Melee_Sml_Att
 bOverrideConcurrency=true
 maxDistance=2800
 soundClassObject=&Fort_WeaponMelee_PSM
 volumeMultiplier=0.5
 concurrencyOverrides {maxCount=1, resolutionRule=PreventNew}
}
